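# MF_DLR_ValueSelectorNormalized is a material function, exposed as "DLR Value Selector" in the Dealership library, that turns a scalar into a 0 or 1 selector.
# The Value input passes through Abs, Round and then Saturate before reaching the single output.
# Negative inputs therefore act like positive ones, and magnitudes that round to 1 or higher clamp to 1.
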MaterialFunction /DealershipSharedLibrary/FX/Materials/MaterialFunctions/MF_DLR_ValueSelectorNormalized
graph {
 properties {
  description="normalize and round a value, return always a positive 0 or 1"
  userExposedCaption=DLR Value Selector
  libraryCategoriesText [Dealership]
 }
 nodes [
  n1 = Abs(None=FunctionInput() {bUsePreviewValueAsDefault=true, inputName=Value, inputType=FunctionInput_Scalar})
  FunctionOutput(None=Saturate(None=Round(None=@n1)))
 ]
}
