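# Material function, named as a saved temporary original, that fakes particles on a mesh, outputting Emissive, Opacity and WorldPosOffset.
# Vertex colors, UV channels 2-3 and Cascade-style dynamic parameters drive WPO for camera-facing spinning quads, explode spread, random size, orbit and dot wobble.
# SubUV flipbook frames vary per UV data, alongside optional rainbow tint and dot-product fake lighting.
# Scalar toggles lerp between features, including custom-texture or rounded-rect opacity masks.
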MaterialFunction /DealershipSharedLibrary/FX/Materials/MaterialFunctions/MF_FakeParticles_SAVE_Original_TEMP
graph {
 properties {bExposeToLibrary=true}
 nodes [
  FunctionInput() {
   bUsePreviewValueAsDefault=true
   desc="Only to  UI"
   description=MF UI Only purpouses
   inputName=***General Parameters***
   inputType=FunctionInput_Bool
  }
  FunctionInput() {
   bUsePreviewValueAsDefault=true
   desc="Only to  UI"
   description=MF UI Only purpouses
   inputName=***WPO Parameters***
   inputType=FunctionInput_Bool
   sortPriority=12
  }
  n1 = NamedRerouteUsage()
  n2 = VectorParameter() {group=General Parameters, parameterName=LightDirection, sortPriority=5, defaultValue=<rgba 0.5 0 0 1>}
  n3 = FunctionInput(Preview=@n2.RGB) {bUsePreviewValueAsDefault=true, inputName=LightDirection, sortPriority=5}
  n4 = Saturate(None=ConstantBiasScale(None=DotProduct(A=@n1, B=@n3)) {bias=0.5})
  n5 = ScalarParameter() {defaultValue=4, group=General Parameters, parameterName=FakeLight Contrast, sortPriority=5}
  n6 = ScalarParameter() {
   defaultValue=1
   desc="Fake lightning default true "
   group=General Parameters
   parameterName="UseFakeLightning? 0=False 1=true"
   sortPriority=4
  }
  n7 = FunctionInput(Preview=@n6) {
   bUsePreviewValueAsDefault=true
   description=Default true
   inputName="UseFakeLightning? 0=False 1=True"
   inputType=FunctionInput_Scalar
   sortPriority=4
  }
  n8 = LinearInterpolate(
   A=NamedRerouteUsage()
   B=ConstantBiasScale(None=Power(Base=@n4, Exp=@n5) {constExponent=4}) {bias=0.1, scale=2}
   Alpha=MaterialFunctionCall!(Value=@n7) {materialFunction=&/MF_DLR_ValueSelectorNormalized}.Result
  )
  n9 = NamedRerouteUsage()
  n10 = FunctionInput(Preview=ScalarParameter() {group=Opacity, parameterName=UseRoundRect?, sortPriority=0}) {bUsePreviewValueAsDefault=true, description=Default false, inputName=UseRoundRect?, inputType=FunctionInput_Scalar, sortPriority=1}
  n11 = MaterialFunctionCall(Value=@n10) {materialFunction=&MF_DLR_ValueSelectorNormalized}
  FunctionOutput(None=LinearInterpolate(A=Multiply(A=@n8, B=@n9), B=@n9, Alpha=@n11.Result)) {outputName=Emissive, sortPriority=1}
  n12 = TextureObjectParameter() {
   group=Opacity
   parameterName=Custom Opacity
   samplerType=SAMPLERTYPE_Grayscale
   texture=/DealershipSharedLibrary/FX/Textures/Masks/T_GradientCircle
  }
  n13 = FunctionInput(Preview=@n12) {bUsePreviewValueAsDefault=true, inputName=Texture, inputType=FunctionInput_Texture2D, sortPriority=2}
  n14 = FunctionInput(Preview=ScalarParameter() {desc="Use ", group=Opacity, parameterName="CustomOpacity? 0=False 1=true", sortPriority=2}) {bUsePreviewValueAsDefault=true, description=Default false, inputName=CustomOpacity, inputType=FunctionInput_Scalar, sortPriority=1}
  n15 = ScalarParameter() {defaultValue=1.3, group=Opacity, parameterName=RoundRect Scale}
  n16 = LinearInterpolate(
   A=LinearInterpolate(
    A=NamedRerouteUsage()
    B=TextureSample(Tex=@n13).RGB
    Alpha=MaterialFunctionCall(Value=@n14) {materialFunction=&MF_DLR_ValueSelectorNormalized}.Result
   )
   B=MaterialFunctionCall!(
    UV Coords=MaterialFunctionCall(Texture Scale=@n15) {materialFunction=&ScaleUVsByCenter}.UVs
    Box Dimensions
    CornerRadius=ScalarParameter() {defaultValue=0.05, group=Opacity, parameterName=RoundRect Corner Radius}
    Sharpness=ScalarParameter() {defaultValue=20, group=Opacity, parameterName=RoundRect Sharpness}
    Center
   ) {materialFunction=/Engine/Functions/Engine_MaterialFunctions01/Gradient/GeneratedRoundRect}.Result
   Alpha=@n11.Result
  )
  n17 = ParticleColor()
  n18 = Multiply(A=@n16, B=@n17.A)
  n19 = NamedRerouteUsage()
  n20 = ScalarParameter() {defaultValue=1, parameterName=ContrastRandomOpacity}
  n21 = Subtract(A=Multiply(A=@n16, B=Power(Base=@n19, Exp=@n20)), B=NamedRerouteUsage())
  n22 = ScalarParameter() {
   controlType=Enumeration
   enumeration=/DealershipSharedLibrary/FX/Materials/Enums/Enum_DLR_TrueFalse
   group=Opacity
   parameterName=UseRandomOpacity
  }
  n23 = Saturate(None=@n22)
  FunctionOutput(None=Saturate(None=LinearInterpolate(A=@n18, B=@n21, Alpha=@n23))) {outputName=Opacity, sortPriority=2}
  n24 = NamedRerouteUsage()
  n25 = NamedRerouteUsage()
  n26 = NamedRerouteUsage()
  n27 = NamedRerouteUsage()
  n28 = Multiply(A=NamedRerouteUsage(), B=Constant() {r=-1})
  n29 = NamedRerouteUsage()
  n30 = Add(
   A=Multiply(A=Multiply(A=@n27, B=@n28), B=ComponentMask(None=@n29) {r=true})
   B=Multiply(A=Multiply(A=@n24, B=@n25), B=ComponentMask(None=@n26) {g=true})
  )
  n31 = WorldPosition()
  n32 = Add(A=Add(A=Subtract(A=@n30, B=@n31.XYZ) {constA=0}, B=NamedRerouteUsage()), B=NamedRerouteUsage())
  n33 = NamedRerouteUsage()
  FunctionOutput(None=NamedRerouteDeclaration(None=Add(A=@n32, B=@n33)) {name=WPO, nodeColor=<rgba 0 0.1355 1 1>}) {outputName=WorldPosOffset, sortPriority=3}
  n34 = NamedRerouteUsage()
  n35 = Power(Base=NamedRerouteUsage(), Exp=Constant() {r=3})
  n36 = DynamicParameter() {parameterIndex=1, paramNames [Dot Movement, Dot Scale, "", ""]}
  n37 = NamedRerouteUsage()
  n38 = NamedRerouteUsage()
  n39 = NamedRerouteUsage()
  n40 = MaterialFunctionCall(
   X=Sine(None=Add(A=@n36.Dot Movement, B=@n37)) {desc=2, period=0.75}
   Y=Cosine(None=Add(A=@n36.Dot Movement, B=@n38)) {desc=2}
   Z=Sine(None=Add(A=@n36.Dot Movement, B=@n39)) {desc=1.333, period=1.3}
  ) {materialFunction=&MakeFloat3}
  n41 = Multiply(
   A=Saturate(None=DynamicParameter() {defaultValue=<rgba 0 1 1 1>, paramNames ["", Spread, Z_Velocity, Scale]}.Spread)
   B=Multiply(A=@n36.Dot Scale, B=ConstantBiasScale(None=@n40.Result) {scale=1})
  )
  n42 = LinearInterpolate(
   A=@n34
   B=Add(A=@n34, B=Multiply(A=@n35, B=@n41))
   Alpha=Round(None=ScalarParameter() {group=WPO, parameterName=IsDotMaterial?, sliderMax=1})
  )
  NamedRerouteDeclaration(None=@n42) {name=Noise Dot Material, nodeColor=<rgba 1 0.2285 0.7045 1>}
  n43 = VertexColor()
  NamedRerouteDeclaration(None=@n43.R) {name=WPO_R, nodeColor=<rgba 1 0.1363 0.2128 1>}
  n44 = TextureCoordinate() {coordinateIndex=2}
  n45 = TextureCoordinate() {coordinateIndex=3}
  n46 = DynamicParameter() {defaultValue=<rgba 0 1 1 1>, paramNames ["", Spread, Z_Velocity, Scale]}
  n47 = MaterialFunctionCall(X=ComponentMask(None=@n44) {r=true}, Y=ComponentMask(None=@n44) {g=true}, Z=Add(A=ComponentMask(None=@n45) {r=true}, B=@n46.Z_Velocity)) {materialFunction=&MakeFloat3}
  n48 = NamedRerouteUsage()
  NamedRerouteDeclaration(None=Multiply(A=Transform(None=@n47.Result) {transformSourceType=TRANSFORMSOURCE_Local}, B=Multiply(A=@n46.Spread, B=@n48))) {name=Explode, nodeColor=<rgba 1 0.0703 0 1>}
  NamedRerouteDeclaration(None=@n43.G) {name=WPO_G, nodeColor=<rgba 0 1 0 1>}
  n49 = MaterialFunctionCall(Float2=TextureCoordinate() {coordinateIndex=2}) {materialFunction=&BreakOutFloat2Components}
  n50 = ComponentMask(None=TextureCoordinate() {coordinateIndex=3}) {r=true}
  NamedRerouteDeclaration(None=TransformPosition(None=MaterialFunctionCall(X=@n49.R, Y=@n49.R, Z=@n50) {materialFunction=&MakeFloat3}.Result) {transformType=TRANSFORMPOSSOURCE_World}) {name=CamCoords, nodeColor=<rgba 1 0 0.8605 1>}
  n51 = NamedRerouteUsage()
  n52 = Constant() {r=0.5}
  MaterialExpressionConstantClamp_1 = Saturate(None=Ceil(None=Subtract(A=@n51, B=@n52)))
  NamedRerouteDeclaration(None=@MaterialExpressionConstantClamp_1) {name=Random WPO Values, nodeColor=<rgba 0.3359 1 0 1>}
  NamedRerouteDeclaration(None=@n43.B) {name=WPO_B, nodeColor=<rgba 0 0 1 1>}
  n53 = Time()
  n54 = FunctionInput(Preview=ScalarParameter() {defaultValue=-1, group=WPO, parameterName=SpinSpeed, sortPriority=19}) {bUsePreviewValueAsDefault=true, inputName=SpinSpeed, inputType=FunctionInput_Scalar, sortPriority=19}
  n55 = LinearInterpolate(A=Multiply(A=NamedRerouteUsage(), B=@n54), B=Multiply(A=NamedRerouteUsage(), B=Abs(None=@n54)), Alpha=NamedRerouteUsage())
  n56 = FunctionInput(Preview=ScalarParameter() {defaultValue=1, desc="Random size default true ", group=WPO, parameterName="UseSpin? 0=False 1=true", sortPriority=18}) {
   bUsePreviewValueAsDefault=true
   description=Default true
   inputName="UseSpin? 0=False 1=True"
   inputType=FunctionInput_Scalar
   sortPriority=18
  }
  NamedRerouteDeclaration(None=LinearInterpolate(
    A=NamedRerouteUsage()
    B=Multiply(A=@n53, B=@n55)
    Alpha=MaterialFunctionCall(Value=@n56) {materialFunction=&MF_DLR_ValueSelectorNormalized}.Result
   )) {name=Spin, nodeColor=<rgba 0.1356 0.0976 0.011 1>}
  n57 = FunctionInput(Preview=ScalarParameter() {defaultValue=1, group=WPO, parameterName=ScaleX, sortPriority=16}) {bUsePreviewValueAsDefault=true, inputName=RotateScaleX, inputType=FunctionInput_Scalar, sortPriority=16}
  n58 = FunctionInput(Preview=ScalarParameter() {defaultValue=1, group=WPO, parameterName=ScaleY, sortPriority=17}) {bUsePreviewValueAsDefault=true, inputName=RotateScaleY, inputType=FunctionInput_Scalar, sortPriority=17}
  n59 = Rotator(Coordinate=Multiply(A=TextureCoordinate(), B=AppendVector(A=@n57, B=@n58)), Time=NamedRerouteUsage()) {speed=3}
  NamedRerouteDeclaration(None=ConstantBiasScale(None=OneMinus(None=@n59)) {bias=-0.5, scale=2}) {name=RotateSprite, nodeColor=<rgba 0.47 0.5259 1 1>}
  MaterialExpressionCameraWorldPosition_0 = CameraPositionWS()
  n60 = Normalize(VectorInput=Subtract(A=@MaterialExpressionCameraWorldPosition_0, B=NamedRerouteUsage()))
  n61 = CrossProduct(A=@n60, B=Normalize(VectorInput=Constant3Vector() {constant=<rgba 0 0 1 0>}))
  NamedRerouteDeclaration(None=Normalize(VectorInput=CrossProduct(A=@n61, B=@n60))) {name=CamCoordsNormalize1, nodeColor=<rgba 0.6875 1 0 1>}
  n62 = Constant()
  n63 = Constant() {r=1}
  n64 = MaterialFunctionCall(Input=@n17.A, Input Low=@n62, Input High=@n63, Target Low=@n63, Target High=@n62) {materialFunction=&RemapValueRange}
  NamedRerouteDeclaration(None=@n64.Result) {name=Remaped 1-0 alpha, nodeColor=<rgba 0 1 0.0391 1>}
  NamedRerouteDeclaration(None=Normalize(VectorInput=@n61)) {name=CamCoordsNormalize2, nodeColor=<rgba 1 0.9711 0 1>}
  n65 = FunctionInput(Preview=ScalarParameter() {defaultValue=3, group=WPO, parameterName=Scale WPO, sortPriority=14}) {bUsePreviewValueAsDefault=true, desc=WPO, description=WPO, inputName=Scale WPO, inputType=FunctionInput_Scalar, sortPriority=14}
  n66 = ScalarParameter() {defaultValue=1, desc="Random size default true ", group=WPO, parameterName="UseRandomSize? 0=False 1=true", sortPriority=13}
  n67 = FunctionInput(Preview=@n66) {
   bUsePreviewValueAsDefault=true
   description=Default true
   inputName="UseRandomSize? 0=False 1=True"
   inputType=FunctionInput_Scalar
   sortPriority=13
  }
  n68 = LinearInterpolate(
   A=@n65
   B=Add(A=NamedRerouteUsage(), B=@n65)
   Alpha=MaterialFunctionCall(Value=@n67) {materialFunction=&MF_DLR_ValueSelectorNormalized}.Result
  )
  n69 = Add(A=@n68, B=ComponentMask(None=TextureCoordinate() {coordinateIndex=3}) {g=true})
  MaterialExpressionMeshEmitterDynamicParameter_0 = DynamicParameter() {defaultValue=<rgba 0 1 1 1>, paramNames ["", Spread, Z_Velocity, Scale]}
  n70 = Ceil(None=Subtract(A=@MaterialExpressionMeshEmitterDynamicParameter_0.Scale, B=NamedRerouteUsage()))
  n71 = ScalarParameter() {defaultValue=1, desc="Random size default true ", group=WPO, parameterName="UseRandomKillSize? 0=False 1=true", sortPriority=15}
  n72 = FunctionInput(Preview=@n71) {
   bUsePreviewValueAsDefault=true
   description=Default true
   inputName="UseRandomKillSize? 0=False 1=True"
   inputType=FunctionInput_Scalar
   sortPriority=15
  }
  n73 = MaterialFunctionCall(Value=@n72) {materialFunction=&MF_DLR_ValueSelectorNormalized}
  NamedRerouteDeclaration(None=LinearInterpolate(
    A=Constant()
    B=Multiply(A=@n68, B=@n69)
    Alpha=LinearInterpolate(A=@MaterialExpressionMeshEmitterDynamicParameter_0.Scale, B=@n70, Alpha=@n73.Result)
   )) {name=Randomize Result, nodeColor=<rgba 0 1 0.6719 1>}
  n74 = FunctionInput(Preview=ScalarParameter() {defaultValue=16, group=General Parameters, parameterName=SubUV_Speed, sortPriority=10}) {bUsePreviewValueAsDefault=true, inputName=SubUV_Speed, inputType=FunctionInput_Scalar, sortPriority=10, previewValue=<16 0 0 1>}
  n75 = NamedRerouteDeclaration(None=ComponentMask(None=TextureCoordinate() {coordinateIndex=3}) {g=true}) {name=UVs Index 3, nodeColor=<rgba 1 0 0.375 1>}
  n76 = Multiply(A=Time(), B=Add(A=@n74, B=@n75))
  n77 = Multiply(A=Constant() {r=-1}, B=@n76)
  n78 = Floor(None=Multiply(A=@n75, B=Constant() {r=2}))
  n79 = FunctionInput(Preview=Add(A=LinearInterpolate(A=@n76, B=@n77, Alpha=@n78), B=NamedRerouteUsage())) {
   bUsePreviewValueAsDefault=true
   description=Custom Frame Based on UVs Index 3
   inputName=CustomFrame
   inputType=FunctionInput_Scalar
   sortPriority=9
  }
  NamedRerouteDeclaration(None=@n79) {name=Frame, nodeColor=<rgba 1 0.8203 0 1>}
  n80 = TextureObjectParameter() {
   group=General Parameters
   parameterName=SubUV Texture
   samplerType=SAMPLERTYPE_LinearColor
   sortPriority=6
   texture=/DealershipSharedLibrary/FX/Textures/Normal/T_Confetti_N
  }
  n81 = FunctionInput(Preview=TextureCoordinate()) {bUsePreviewValueAsDefault=true, description=rows and columns, inputName=CustomUvs, inputType=FunctionInput_Vector2, sortPriority=7}
  n82 = VectorParameter() {group=General Parameters, parameterName=SubImages, sortPriority=8, channelNames {a="", b="", g=Columns, r=Rows}, defaultValue=<rgba 4 4 0 0>}
  n83 = FunctionInput(Preview=MaterialFunctionCall(X=@n82.Rows, Y=@n82.Rows) {materialFunction=&MakeFloat2}.Result) {
   bUsePreviewValueAsDefault=true
   description=rows and columns default 4x4
   inputName=SubImages
   inputType=FunctionInput_Vector2
   sortPriority=8
  }
  n84 = MaterialFunctionCall!(
   Texture=FunctionInput(Preview=@n80) {bUsePreviewValueAsDefault=true, inputName=SubUV_Texture, inputType=FunctionInput_Texture2D, sortPriority=6}
   UVs=@n81
   SubImages=@n83
   Frame=NamedRerouteUsage()
  ) {materialFunction=/Engine/Functions/Engine_MaterialFunctions03/Texturing/SubUV_Function}
  NamedRerouteDeclaration(None=@n84.Alpha) {name=SubUVAlpha, nodeColor=<rgba 0 0.7187 1 1>}
  NamedRerouteDeclaration(None=@n84.RGB) {name=SubUvRGB, nodeColor=<rgba 1 0.9609 0 1>}
  n85 = TextureSample(UVs=NamedRerouteUsage()) {texture=/DealershipSharedLibrary/FX/Textures/Masks/T_Rainbow_D}
  n86 = ParticleColor()
  n87 = FunctionInput(Preview=ScalarParameter() {desc=Default false, group=General Parameters, parameterName="UseRandomColor? 0=False 1=true", sortPriority=3}) {bUsePreviewValueAsDefault=true, inputName="Use RandomColor 0=False 1=True", inputType=FunctionInput_Scalar, sortPriority=3}
  NamedRerouteDeclaration(None=LinearInterpolate(
    A=@n86.RGB
    B=Multiply(A=@n86.RGB, B=@n85.RGB) {desc=Multiply particle color for intensity}
    Alpha=MaterialFunctionCall(Value=@n87) {materialFunction=&MF_DLR_ValueSelectorNormalized}.Result
   )) {name=RandomColor, nodeColor=<rgba 0.8047 1 0 1>}
  n88 = FunctionInput(Preview=ScalarParameter() {defaultValue=1, group=WPO, parameterName=OrbitSpeed, sortPriority=21}) {bUsePreviewValueAsDefault=true, inputName=RotateScaleX_1, inputType=FunctionInput_Scalar, sortPriority=21}
  n89 = Multiply(A=Multiply(A=Time(), B=NamedRerouteUsage()), B=LinearInterpolate(A=@n88, B=Abs(None=@n88)))
  n90 = Sine(None=@n89)
  n91 = Cosine(None=@n89) {period=0.5}
  n92 = ScalarParameter() {defaultValue=1, group=WPO, parameterName="Orbit 0=False 1=True x=Intensity", sortPriority=20}
  NamedRerouteDeclaration(None=Multiply(
    A=MaterialFunctionCall(X=@n90, Y=@n91) {materialFunction=&MakeFloat3}.Result
    B=FunctionInput(Preview=@n92) {bUsePreviewValueAsDefault=true, inputName="Orbit 0=False X=Intensity", inputType=FunctionInput_Scalar, sortPriority=20}
   )) {name=Orbit, nodeColor=<rgba 0.7109 0 1 1>}
  Reroute(None=@n46.R)
  Reroute(None=@n42)
  Reroute(None=@n64.Result)
 ]
}
