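# Unlit scratch material of reference node graphs, apparently for FX material functions, with nothing wired to a material output.
# Separate snippets cover smoothstep, range normalization, bias and scale, UV edge masks, and time-driven rotation about an axis.
# A selector switches between texture, world XY and particle macro UVs.
# Primary and secondary mask chains build gradient, sine and triangle shapes from rotated UVs, then pick channels with static mask parameters and remap them.
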Material /DealershipSharedLibrary/FX/Materials/MaterialFunctions/M_MaterialGraph_References
graph {
 properties {bAutomaticallySetUsageInEditor=false, shadingModel=MSM_Unlit, usedShadingModels=Unlit}
 nodes [
  n1 = Constant() {r=3}
  n2 = ScalarParameter() {parameterName=Input}
  n3 = ScalarParameter() {defaultValue=0.1, parameterName=Input Min}
  n4 = ScalarParameter() {defaultValue=0.9, parameterName=Input Max}
  MaterialExpressionConstantClamp_4 = Saturate(None=Divide(A=Subtract(A=@n2, B=@n3), B=Subtract(A=@n4, B=@n3)))
  n5 = Multiply(A=@MaterialExpressionConstantClamp_4, B=Constant() {r=2})
  n6 = Multiply(A=Multiply(A=Subtract(A=@n1, B=@n5), B=@MaterialExpressionConstantClamp_4), B=@MaterialExpressionConstantClamp_4)
  AppendVector(A=AppendVector(A=@n6, B=@n6), B=@n6)
  n7 = ScalarParameter() {defaultValue=1, parameterName=UV Selector}
  n8 = TextureCoordinate()
  n9 = If(A=@n7, "A > B"=WorldPosition().XY, "A == B"=@n8, "A < B"=@n8) {constB=0.99}
  If(A=@n7, "A > B"=ParticleMacroUV(), "A == B"=@n9, "A < B"=@n9) {constB=1.99}
  n10 = TextureCoordinate()
  n11 = VectorParameter() {
   desc="RG controls normalized rotation angle (0 - 1) for Primary and Secondary mask respectively."
   group=Masks
   parameterName=Masks Rotation Angle
   sortPriority=8
   defaultValue=<rgba 0 0 0 1>
  }
  n12 = MaterialFunctionCall(UVs=@n10, "Rotation Angle (0-1)"=@n11.G) {materialFunction=&CustomRotator}
  n13 = ConstantBiasScale(None=@n12.Rotated Values) {bias=-0.5, scale=2}
  n14 = Abs(None=ComponentMask(None=@n13) {r=true})
  n15 = OneMinus(None=ComponentMask(None=@n13) {g=true})
  n16 = Multiply(A=@n14, B=@n15)
  n17 = Multiply(A=OneMinus(None=@n14), B=@n15)
  n18 = VectorParameter() {
   desc=R controls Input Low. G controls Input High. B controls Target Low. A controls Target High.
   group=Masks
   parameterName=Secondary Mask Set 2 - Remap
   sortPriority=7
   defaultValue=<rgba 0 1 0 1>
  }
  MaterialFunctionCall(
   Input=StaticComponentMaskParameter(None=MaterialFunctionCall(X=@n16, Y=@n17) {materialFunction=&MakeFloat2}.Result) {defaultR=true, group=Masks, parameterName=Secondary Mask Set 2 - Selector, sortPriority=6}
   Input Low=@n18.R
   Input High=@n18.R
   Target Low=@n18.R
   Target High=@n18.R
  ) {materialFunction=&RemapValueRange}
  n19 = ComponentMask(None=@n12.Rotated Values) {r=true}
  n20 = Sine(None=@n19) {period=2}
  n21 = OneMinus(None=@n20)
  n22 = Multiply(A=@n20, B=Sine(None=ComponentMask(None=@n12.Rotated Values) {g=true}) {period=2})
  n23 = VectorParameter() {
   desc=R controls Input Low. G controls Input High. B controls Target Low. A controls Target High.
   group=Masks
   parameterName=Secondary Mask Set 1 - Remap
   sortPriority=5
   defaultValue=<rgba 0 1 0 1>
  }
  MaterialFunctionCall(
   Input=StaticComponentMaskParameter(None=MaterialFunctionCall(X=@n19, Y=@n20, Z=@n21, A=@n22) {materialFunction=&MakeFloat4}.Result) {defaultR=true, group=Masks, parameterName=Secondary Mask Set 1 - Selector, sortPriority=4}
   Input Low=@n23.R
   Input High=@n23.R
   Target Low=@n23.R
   Target High=@n23.R
  ) {materialFunction=&RemapValueRange}
  n24 = MaterialFunctionCall(UVs=@n10, "Rotation Angle (0-1)"=@n11.R) {materialFunction=&CustomRotator}
  n25 = ConstantBiasScale(None=@n24.Rotated Values) {bias=-0.5, scale=2}
  n26 = Abs(None=ComponentMask(None=@n25) {r=true})
  n27 = OneMinus(None=ComponentMask(None=@n25) {g=true})
  n28 = Multiply(A=@n26, B=@n27)
  n29 = Multiply(A=OneMinus(None=@n26), B=@n27)
  n30 = VectorParameter() {
   desc=R controls Input Low. G controls Input High. B controls Target Low. A controls Target High.
   group=Masks
   parameterName=Primary Mask Set 2 - Remap
   sortPriority=3
   defaultValue=<rgba 0 1 0 1>
  }
  MaterialFunctionCall(
   Input=StaticComponentMaskParameter(None=MaterialFunctionCall(X=@n28, Y=@n29) {materialFunction=&MakeFloat2}.Result) {defaultR=true, group=Masks, parameterName=Primary Mask Set 2 - Selector, sortPriority=2}
   Input Low=@n30.R
   Input High=@n30.R
   Target Low=@n30.R
   Target High=@n30.R
  ) {materialFunction=&RemapValueRange}
  n31 = ComponentMask(None=@n24.Rotated Values) {r=true}
  n32 = Sine(None=@n31) {period=2}
  n33 = OneMinus(None=@n32)
  n34 = Multiply(A=@n32, B=Sine(None=ComponentMask(None=@n24.Rotated Values) {g=true}) {period=2})
  n35 = VectorParameter() {
   desc=R controls Input Low. G controls Input High. B controls Target Low. A controls Target High.
   group=Masks
   parameterName=Primary Mask Set 1 - Remap
   sortPriority=1
   defaultValue=<rgba 0 1 0 1>
  }
  MaterialFunctionCall(
   Input=StaticComponentMaskParameter(None=MaterialFunctionCall(X=@n31, Y=@n32, Z=@n33, A=@n34) {materialFunction=&MakeFloat4}.Result) {defaultR=true, group=Masks, parameterName="Primary Mask Set 1 -  Selector", sortPriority=0}
   Input Low=@n35.R
   Input High=@n35.R
   Target Low=@n35.R
   Target High=@n35.R
  ) {materialFunction=&RemapValueRange}
  Multiply(
   A=Add(A=VectorParameter() {parameterName=Input, defaultValue=<rgba 1 1 1 1>}.RGB, B=ScalarParameter() {defaultValue=1, parameterName=Bias})
   B=ScalarParameter() {defaultValue=1, parameterName=Scale}
  )
  n36 = MaterialFunctionCall(Vector=VectorParameter() {parameterName=Axis, defaultValue=<rgba 1 0 0 0>}.RGB, Default=Constant3Vector() {constant=<rgba 1 0 0 0>}) {materialFunction=&SafeNormalize}
  n37 = CrossProduct(A=@n36.Result, B=Constant3Vector() {constant=<rgba 0 0 1 0>})
  n38 = Constant3Vector() {constant=<rgba 1 0 0 0>}
  RotateAboutAxis(
   NormalizedRotationAxis=@n36.Result
   RotationAngle=Time()
   PivotPoint=Constant3Vector()
   Position=Multiply(
    A=MaterialFunctionCall(Vector=@n37, Default=@n38) {materialFunction=&SafeNormalize}.Result
    B=ScalarParameter() {defaultValue=5, parameterName=Amplitude}
   ) {constB=10}
  )
  n39 = ScalarParameter() {parameterName=Function Input}
  n40 = ScalarParameter() {parameterName=Input Low}
  n41 = ScalarParameter() {parameterName=Input High}
  Saturate(None=Divide(A=Subtract(A=@n39, B=@n40), B=Subtract(A=@n41, B=@n40)))
  n42 = MaterialFunctionCall(Float2=TextureCoordinate()) {materialFunction=&BreakOutFloat2Components}
  n43 = Multiply(A=OneMinus(None=@n42.R), B=@n42.R)
  Saturate(None=Divide(A=Multiply(A=@n43) {constB=4}, B=ScalarParameter() {defaultValue=1, parameterName=Mask Distance X}))
  n44 = Multiply(A=OneMinus(None=@n42.G), B=@n42.G)
  Saturate(None=Divide(A=Multiply(A=@n44) {constB=4}, B=ScalarParameter() {defaultValue=1, parameterName=Mask Distance Y}))
  VectorParameter() {parameterName="Custom World Position (Vector 2)"}
 ]
}
