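# A reusable dissolve material function that thresholds an input alpha at a Cutoff to output an OpacityMask, intended to be multiplied with other alpha operations.
# A SmoothStep starts at Cutoff, and the Soft Transition static switch either collapses its upper edge onto Cutoff for a hard edge or widens it.
# The soft upper edge is Cutoff plus Transition Range, or one minus Cutoff when Normalized Alpha Range is enabled.
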MaterialFunction /DealershipSharedLibrary/FX/Materials/MaterialFunctions/MF_AlphaThreshold_Smoothstep
graph {
 properties {bExposeToLibrary=true, description="AlphaThreshold (Dissolve) with smooth edge control (SmoothStep based))"}
 nodes [
  n1 = FunctionInput() {
   bUsePreviewValueAsDefault=true
   description=Alpha cutoff value
   inputName=Cutoff
   inputType=FunctionInput_Scalar
   sortPriority=2
   previewValue=<0.23 0 0 1>
  }
  n2 = FunctionInput() {
   bUsePreviewValueAsDefault=true
   description=Alpha transition range
   inputName=Transition Range
   inputType=FunctionInput_Scalar
   sortPriority=4
   previewValue=<0.15 0 0 1>
  }
  n3 = FunctionInput(Preview=StaticBool()) {
   bUsePreviewValueAsDefault=true
   description="Ignores Transition Range, remaps the alpha to always be between 0-1 after the Cutoff"
   inputName=Normalized Alpha Range
   inputType=FunctionInput_StaticBool
   sortPriority=5
  }
  n4 = StaticSwitch(True=OneMinus(None=@n1), False=Add(A=@n1, B=@n2), Value=@n3)
  n5 = FunctionInput(Preview=StaticBool()) {
   bUsePreviewValueAsDefault=true
   description=Use a soft transition from Cutoff value
   inputName=Soft Transition
   inputType=FunctionInput_StaticBool
   sortPriority=3
  }
  n6 = FunctionInput() {
   bUsePreviewValueAsDefault=true
   description=Alpha texture
   inputName=Alpha
   inputType=FunctionInput_Scalar
   sortPriority=1
   previewValue=<1 0 0 1>
  }
  FunctionOutput(None=SmoothStep(Min=@n1, Max=StaticSwitch(True=@n4, False=@n1, Value=@n5), Value=@n6)) {description=To be multiplied with other Alpha operations, outputName=OpacityMask}
 ]
}
