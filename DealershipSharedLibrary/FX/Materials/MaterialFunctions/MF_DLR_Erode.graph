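# Dealership FX material function, exposed to the library as DLR Erosion, that erodes a scalar mask texture by a 0-1 erosion amount.
# It remaps Erosion using Width to set a threshold, then smoothsteps the mask across that width for a soft-edged result.
# A second Edge output repeats this with Erosion offset by Edge Width, then saturates and ceils it into a hard binary mask.
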MaterialFunction /DealershipSharedLibrary/FX/Materials/MaterialFunctions/MF_DLR_Erode
graph {
 properties {
  bExposeToLibrary=true
  description="Basic erosion using an erosion amount [0-1] and a  mask texture sample [0-1]"
  userExposedCaption=DLR Erosion
  libraryCategoriesText [Dealership]
 }
 nodes [
  n1 = FunctionInput(Preview=TextureSample() {samplerType=SAMPLERTYPE_LinearColor, texture=/Game/Global/Textures/Default/Noise/T_CloudNoise_01}.R) {desc="Mask texture [0-1]", inputName=Mask, inputType=FunctionInput_Scalar, sortPriority=1}
  n2 = FunctionInput(Preview=MaterialFunctionCall() {materialFunction=&DebugTimeSine}.Result) {bUsePreviewValueAsDefault=true, desc="Erosion amount  [0-1]", inputName=Erosion, inputType=FunctionInput_Scalar}
  n3 = FunctionInput() {
   bUsePreviewValueAsDefault=true
   desc=Smoothstep width
   inputName=Width
   inputType=FunctionInput_Scalar
   sortPriority=2
   previewValue=<0.25 0 0 1>
  }
  n4 = Subtract(A=@n2, B=@n3)
  n5 = MaterialFunctionCall(Input=@n2, Target Low=@n4) {materialFunction=&RemapValueRange}
  FunctionOutput(None=MaterialFunctionCall(Alpha=@n1, Min=@n5.Result, Max=Add(A=@n5.Result, B=@n3) {constB=0.25}) {materialFunction=&SmoothStep}.Result)
  n6 = FunctionInput() {bUsePreviewValueAsDefault=true, desc=Smoothstep width, inputName=Edge Width, inputType=FunctionInput_Scalar, sortPriority=2}
  n7 = MaterialFunctionCall(Input=Subtract(A=@n2, B=Multiply(A=@n6) {constB=-1}) {constB=0.1}, Target Low=@n4) {materialFunction=&RemapValueRange}
  n8 = MaterialFunctionCall(Alpha=@n1, Min=@n7.Result, Max=Add(A=@n7.Result, B=@n3) {constB=0.25}) {materialFunction=&SmoothStep}
  FunctionOutput(None=Ceil(None=Saturate(None=@n8.Result))) {outputName=Edge}
 ]
}
