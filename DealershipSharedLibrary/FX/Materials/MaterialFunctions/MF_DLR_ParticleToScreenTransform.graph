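# Material function that outputs World Position Offset to pull sprite or mesh particles in front of the camera and size them relative to the screen.
# It places the particle at CameraOffset along a blend of camera and camera-to-particle directions, plus a view-space offset from the particle-to-actor Y difference.
# Vertices are normalized by particle size or local bounds' Y extent, then rescaled using view size and FOV.
# FitScreenPercentage weights this reshaping, and ScreenSizePercentage scales it.
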MaterialFunction /DealershipSharedLibrary/FX/Materials/MaterialFunctions/MF_DLR_ParticleToScreenTransform
graph {
 properties {libraryCategoriesText [Dealership]}
 nodes [
  n1 = MaterialFunctionCall(Float3=Subtract(A=ParticlePositionWS(), B=ActorPositionWS())) {materialFunction=&BreakOutFloat3Components}
  n2 = Transform(None=MaterialFunctionCall(X=@n1.G, Y=@n1.G, Z=@n1.G) {materialFunction=&MakeFloat3}.Result) {transformSourceType=TRANSFORMSOURCE_View}
  n3 = CameraPositionWS()
  n4 = ParticlePositionWS()
  n5 = MaterialFunctionCall() {materialFunction=&CameraDirectionVector}
  n6 = Normalize(VectorInput=Subtract(A=@n4, B=@n3))
  n7 = FunctionInput(Preview=Constant()) {bUsePreviewValueAsDefault=true, inputName=Align Offset to Particle, inputType=FunctionInput_Scalar, sortPriority=1}
  n8 = FunctionInput(Preview=Constant() {r=100}) {bUsePreviewValueAsDefault=true, inputName=CameraOffset, inputType=FunctionInput_Scalar, sortPriority=1}
  n9 = Add(A=Subtract(A=@n3, B=@n4), B=Multiply(A=LinearInterpolate(A=@n5.Result, B=@n6, Alpha=@n7), B=@n8))
  n10 = WorldPosition()
  n11 = ParticlePositionWS()
  n12 = ViewSize()
  n13 = Multiply(A=@n12) {constB=0.5}
  n14 = Subtract(A=WorldPosition().XYZ, B=ParticlePositionWS())
  n15 = MaterialFunctionCall(Float3=Max(A=Bounds() {type=MEILB_PreSkinnedLocal}.Half Extents) {constB=0.01}) {materialFunction=&BreakOutFloat3Components}
  n16 = MaterialFunctionCall(X=@n15.G, Y=@n15.G, Z=@n15.G) {materialFunction=&MakeFloat3}
  n17 = Multiply(A=AppendVector(A=ParticleSize(), B=Constant())) {constB=0.5}
  n18 = FunctionInput() {bUsePreviewValueAsDefault=true, inputName=Is Sprite, inputType=FunctionInput_Scalar, sortPriority=4}
  n19 = ComponentMask(None=Divide(
    A=Transform(None=@n14) {transformSourceType=TRANSFORMSOURCE_World, transformType=TRANSFORM_Camera}
    B=LinearInterpolate(A=@n16.Result, B=@n17, Alpha=@n18)
   )) {g=true, r=true}
  n20 = ViewProperty() {property=MEVP_TanHalfFieldOfView}
  n21 = Multiply(A=Multiply(A=ComponentMask(None=@n20.InvProperty) {r=true}, B=ComponentMask(None=@n12) {r=true})) {constB=0.5}
  n22 = AppendVector(A=Multiply(A=Multiply(A=@n19, B=@n13), B=Divide(A=@n8, B=@n21)), B=Constant())
  n23 = Constant() {r=1}
  n24 = Add(
   A=Multiply(A=Subtract(A=@n10.XYZ, B=@n11), B=Constant() {r=-1}) {constB=-1}
   B=Multiply(
    A=Transform(None=@n22) {transformSourceType=TRANSFORMSOURCE_Camera}
    B=FunctionInput(Preview=@n23) {bUsePreviewValueAsDefault=true, inputName=ScreenSizePercentage}
   )
  ) {constB=0}
  n25 = FunctionInput(Preview=Constant() {r=1}) {bUsePreviewValueAsDefault=true, inputName=FitScreenPercentage, inputType=FunctionInput_Scalar}
  FunctionOutput(None=Add(A=Add(A=@n2, B=@n9), B=Multiply(A=@n24, B=@n25)) {constB=0}) {outputName=WorldPositionOffset}
 ]
}
