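# Material function for Dealership full-screen transitions, outputting gradient, curve color and alpha, edge color, and a WPO that attaches geometry to the camera.
# A static switch picks a camera-direction or world-distance mask around a center position for a SphereMask, whose result indexes a curve atlas.
# Noise in world-aligned screen coordinates provides distortion.
# Static switches source phase, hardness, distortion, position and edge color from MPC_FullScreenTransition or material parameters.
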MaterialFunction /DealershipSharedLibrary/FX/Materials/MaterialFunctions/MF_DLR_FullScreenTransition_Alpha
graph {
 properties {previewBlendMode=BLEND_Translucent, userExposedCaption=Full Screen Transition, libraryCategoriesText []}
 nodes [
  n1 = Constant() {r=1}
  FunctionOutput(None=MaterialFunctionCall!(
    ScreenSizePercentage=Constant() {r=1}
    FitScreenPercentage=Constant() {r=1}
    Align Offset to Particle
    CameraOffset=Constant() {r=100}
    Is Sprite=FunctionInput(Preview=@n1) {bUsePreviewValueAsDefault=true, inputName=Is Sprite, inputType=FunctionInput_Scalar}
   ) {materialFunction=&/MF_DLR_ParticleToScreenTransform}.WorldPositionOffset) {outputName=Attach to camera WPO, sortPriority=3}
  FunctionOutput(None=StaticSwitch(
    True=CollectionParameter() {collection=/DealershipSharedLibrary/FX/Materials/MasterMaterials/PostProcess/MPC_FullScreenTransition, parameterName=Edge - Color}
    False=VectorParameter() {group=Edge, parameterName=Edge - Color, defaultValue=<rgba 5 5 5 1>}.RGBA
    Value=NamedRerouteUsage()
   )) {outputName=Edge Color, sortPriority=3}
  n2 = NamedRerouteUsage()
  n3 = SphereMask(
   A=StaticSwitchParameter(True=NamedRerouteUsage(), False=NamedRerouteUsage()) {group=Transition, parameterName=Transition - Use World Space Mask, sortPriority=0}
   B=NamedRerouteUsage()
   Radius=If(
    A=ScalarParameter() {group=Transition, parameterName=Transition - Use Phase Anim Curve On/Off, sliderMax=1, sortPriority=50}
    B=Constant()
    "A > B"=ScalarParameter() {group=Transition, parameterName=Transition - Use Phase Anim Curve, sortPriority=50}
    "A == B"=@n2
    "A < B"=@n2
   )
   Hardness=NamedRerouteUsage()
  ) {hardnessPercent=0}
  FunctionOutput(None=@n3) {outputName=Base Gradient}
  n4 = TextureSampleParameter2D(UVs=AppendVector(A=@n3, B=Constant() {r=1})) {
   group=Transition
   parameterName=Transition - Curves
   samplerType=SAMPLERTYPE_LinearColor
   sortPriority=10
   texture=/DealershipSharedLibrary/FX/Materials/MasterMaterials/FullScreenTransition/Curves/Curve_FullScreen_Atlas
  }
  FunctionOutput(None=@n4.RGB) {outputName=Curve - RGB, sortPriority=-1}
  FunctionOutput(None=@n4.A) {outputName=Curve - Alpha, sortPriority=-1}
  MaterialFunctionCall() {materialFunction=&DebugTimeSine}
  n5 = CollectionParameter() {
   collection=&MPC_FullScreenTransition
   parameterName=Transition - Distortion Scale
  }
  NamedRerouteDeclaration(None=StaticSwitch(
    True=@n5
    False=VectorParameter() {group=Transition, parameterName=Transition Distortion - Scale, defaultValue=<rgba 0 0 0 1>}.RGB
    Value=NamedRerouteUsage()
   )) {name=Transition Distortion Scale, nodeColor=<rgba 1 0.3047 0 1>}
  NamedRerouteDeclaration(None=FunctionInput(Preview=StaticBool()) {bUsePreviewValueAsDefault=true, inputName=Use MPC, inputType=FunctionInput_StaticBool}) {name=Use MPC Bool, nodeColor=<rgba 1 0.0165 0 1>}
  n6 = Normalize(VectorInput=Subtract(A=CameraPositionWS(), B=NamedRerouteUsage()))
  n7 = CameraVectorWS()
  NamedRerouteDeclaration(None=Multiply(A=Distance(A=@n6, B=@n7)) {constB=0.25}) {name=Screen Space Mask, nodeColor=<rgba 0 0.2969 1 1>}
  n8 = CollectionParameter() {collection=&MPC_FullScreenTransition, parameterName=Transition - Phase}
  n9 = VectorParameter() {group=Transition, parameterName=Transition - Attributes, channelNames {a="", b=Distortion, g=Hardness, r=Phase}, defaultValue=<rgba 0 0 0 1>}
  n10 = NamedRerouteUsage()
  NamedRerouteDeclaration(None=StaticSwitch(True=@n8, False=@n9.Phase, Value=@n10)) {name=Transition - Phase, nodeColor=<rgba 0 1 0.625 1>}
  n11 = CollectionParameter() {collection=&MPC_FullScreenTransition, parameterName=Transition - Hardness}
  NamedRerouteDeclaration(None=StaticSwitch(True=@n11, False=@n9.Hardness, Value=@n10)) {name=Transition - Hardness, nodeColor=<rgba 0 1 0.625 1>}
  n12 = CollectionParameter() {collection=&MPC_FullScreenTransition, parameterName=Transition - Distortion}
  NamedRerouteDeclaration(None=StaticSwitch(True=@n12, False=@n9.Distortion, Value=@n10)) {name=Transition - Distortion, nodeColor=<rgba 0 1 0.625 1>}
  n13 = MaterialFunctionCall() {materialFunction=&WorldPositionBehindTranslucency}
  n14 = WorldPosition()
  n15 = FunctionInput(Preview=StaticBool()) {bUsePreviewValueAsDefault=true, inputName=Use World Position Behind Translucency, inputType=FunctionInput_StaticBool}
  n16 = Divide(
   A=Distance(A=StaticSwitch(True=@n13.Result, False=@n14.XYZ, Value=@n15), B=NamedRerouteUsage())
   B=ScalarParameter() {defaultValue=250000, group=Transition, parameterName=World Space Gradient - Radius, sortPriority=50}
  )
  n17 = OneMinus(None=ComponentMask(None=ScreenPosition().ViewportUV) {g=true})
  n18 = Floor(None=@n16)
  NamedRerouteDeclaration(None=LinearInterpolate(A=@n16, B=Multiply(A=@n17) {constB=0.5}, Alpha=Saturate(None=@n18))) {name=World Space Mask, nodeColor=<rgba 0 0.3203 1 1>}
  n19 = CollectionParameter() {collection=&MPC_FullScreenTransition, parameterName=Position}
  NamedRerouteDeclaration(None=StaticSwitch(
    True=ComponentMask(None=@n19) {b=true, g=true, r=true}
    False=VectorParameter() {group=Transition, parameterName=Transition Distortion - Position, defaultValue=<rgba 0 0 0 1>}.RGB
    Value=NamedRerouteUsage()
   )) {name=Center Position, nodeColor=<rgba 1 0.3047 0 1>}
  n20 = NamedRerouteUsage()
  n21 = NamedRerouteUsage()
  n22 = ScreenPosition()
  n23 = TextureSampleParameter2D!(
   UVs=Multiply(
    A=MaterialFunctionCall!(Center Point World Position=@n20) {materialFunction=/Engine/Functions/Engine_MaterialFunctions02/UVs/WorldSpaceAlignedScreenCoordinates}.Offset and Scaled UVS
    B=ComponentMask(None=@n21) {g=true, r=true}
   )
   "DDX(UVs)"=DDX(Value=@n22.ViewportUV)
   "DDY(UVs)"=DDY(Value=@n22.ViewportUV)
   Apply View MipBias
  ) {
   group=Transition
   mipValueMode=TMVM_Derivative
   parameterName=Transition Distortion - Texture
   sortPriority=40
   texture=&T_NoisePack_01
  }
  n24 = NamedRerouteUsage()
  NamedRerouteDeclaration(None=Multiply(
    A=ChannelMaskParameter(None=@n23.RGBA) {group=Transition, parameterName=Transition Distortion - Channel, sortPriority=41}.R
    B=Multiply(A=@n24) {constB=0.1}
   )) {name=Screen Space Transition Distortion, nodeColor=<rgba 1 0 0.7266 1>}
 ]
}
