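# Material function that supplies Dealership painted-variant data to materials: primary, secondary and tertiary colors, an Is Painted value and a variant ID.
# Each color output lerps from an input (with a default preview) to its PaintedColor parameter, with primary split into RGB and A.
# IsPaintedColor, ColorVariantID and PaintedColor1 read Custom Primitive Data for per-primitive variation.
# PaintedColor2 and PaintedColor3 are ordinary material parameters.
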MaterialFunction /DealershipSharedLibrary/FX/Materials/MaterialFunctions/MF_DLR_PaintedVariantsData
graph {
 properties {description=Useful data related to Dealership Painted Variants, userExposedCaption=DLR Painted Variant Data, libraryCategoriesText [""]}
 nodes [
  n1 = VectorParameter() {bUseCustomPrimitiveData=true, parameterName=PaintedColor1, primitiveDataIndex=2, defaultValue=<rgba 1 0 0.8983 1>}
  n2 = LinearInterpolate(
   A=FunctionInput(Preview=Constant() {r=1}) {bUsePreviewValueAsDefault=true, inputName=Primary Color, inputType=FunctionInput_Vector4, sortPriority=-2}
   B=AppendVector(A=@n1.RGB, B=@n1.RGB)
   Alpha=NamedRerouteUsage()
  )
  FunctionOutput(None=ComponentMask(None=@n2) {b=true, g=true, r=true}) {outputName=Primary RGB, sortPriority=-2}
  n3 = ScalarParameter() {bUseCustomPrimitiveData=true, parameterName=IsPaintedColor, primitiveDataIndex=1}
  FunctionOutput(None=NamedRerouteDeclaration(None=@n3) {name=Is Painted, nodeColor=<rgba 0 0.9297 1 1>}) {outputName=Is Painted, sortPriority=3}
  n4 = Constant() {r=1}
  n5 = VectorParameter() {parameterName=PaintedColor2, primitiveDataIndex=2, defaultValue=<rgba 1 0 0.7479 1>}
  FunctionOutput(None=LinearInterpolate(
    A=FunctionInput(Preview=@n4) {bUsePreviewValueAsDefault=true, inputName="Secondary  Color", inputType=FunctionInput_Vector4, sortPriority=1}
    B=AppendVector(A=@n5.RGB, B=@n5.RGB)
    Alpha=NamedRerouteUsage()
   )) {outputName="Secondary  RGBA", sortPriority=1}
  FunctionOutput(None=ScalarParameter() {bUseCustomPrimitiveData=true, parameterName=ColorVariantID}) {outputName=Variand ID, sortPriority=4}
  n6 = Constant() {r=1}
  n7 = VectorParameter() {parameterName=PaintedColor3, primitiveDataIndex=2, defaultValue=<rgba 1 0 0.7479 1>}
  FunctionOutput(None=LinearInterpolate(
    A=FunctionInput(Preview=@n6) {bUsePreviewValueAsDefault=true, inputName=Tertiary Color, inputType=FunctionInput_Vector4, sortPriority=1}
    B=AppendVector(A=@n7.RGB, B=@n7.RGB)
    Alpha=NamedRerouteUsage()
   )) {outputName=Tertiary RGBA, sortPriority=2}
  FunctionOutput(None=ComponentMask(None=@n2) {a=true}) {outputName=Primary A, sortPriority=-1}
  Reroute(None=@n3)
 ]
}
