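# Material function in the shared FX library that outputs a linear UV gradient with invert, contrast and brightness controls.
# The U/V selector lerp takes the R (U) coordinate on both inputs, so the gradient always follows U.
# The gradient is optionally inverted, raised to a contrast power, multiplied by brightness and saturated.
# A UseDynamicValues? enum parameter switches contrast and brightness between standard and dynamic inputs through If nodes.
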MaterialFunction /DealershipSharedLibrary/FX/Materials/MaterialFunctions/MF_LinearGradientPlus
graph {nodes [
  n1 = MaterialFunctionCall(Float2=TextureCoordinate()) {materialFunction=&BreakOutFloat2Components}
  n2 = LinearInterpolate(
   A=@n1.R
   B=@n1.R
   Alpha=FunctionInput(Preview=Constant()) {bUsePreviewValueAsDefault=true, description="Default = 0", inputName="Gradient U=0 V=1", inputType=FunctionInput_Scalar}
  )
  n3 = FunctionInput(Preview=Constant()) {
   bUsePreviewValueAsDefault=true
   description="Default = 0"
   inputName="Invert gradient 0=False 1=True"
   inputType=FunctionInput_Scalar
   sortPriority=1
  }
  n4 = ScalarParameter() {
   controlType=Enumeration
   desc=Dynamic values for Contrast and brightness
   enumeration=/DealershipSharedLibrary/FX/Materials/Enums/Enum_DLR_TrueFalse
   group=Opacity
   parameterName=UseDynamicValues?
  }
  n5 = Constant() {r=1}
  n6 = FunctionInput(Preview=@n5) {
   bUsePreviewValueAsDefault=true
   description="Default = 1"
   inputName="Dynamic Contrast (Power)"
   inputType=FunctionInput_Scalar
   sortPriority=4
  }
  n7 = FunctionInput(Preview=@n5) {bUsePreviewValueAsDefault=true, description="Default = 1", inputName="Contrast (Power)", inputType=FunctionInput_Scalar, sortPriority=2}
  n8 = Power(Base=LinearInterpolate(A=@n2, B=OneMinus(None=@n2), Alpha=@n3), Exp=If(A=@n4, "A > B"=@n6, "A == B"=@n7, "A < B"=@n7) {constB=0.5})
  n9 = FunctionInput(Preview=@n5) {
   bUsePreviewValueAsDefault=true
   description="Default = 1"
   inputName="Dynamic Brightness (Multiply)"
   inputType=FunctionInput_Scalar
   sortPriority=5
  }
  n10 = FunctionInput(Preview=@n5) {bUsePreviewValueAsDefault=true, description="Default = 1", inputName="Brightness (Multiply)", inputType=FunctionInput_Scalar, sortPriority=3}
  n11 = If(A=@n4, "A > B"=@n9, "A == B"=@n10, "A < B"=@n10) {constB=0.5}
  FunctionOutput(None=Saturate(None=Multiply(A=@n8, B=@n11)))
 ]}
