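# Material function that remaps a Color input by a scalar Shift (default 1), blending in a monochrome component.
# It projects Color onto the normalized white axis for a grey, then adds Color, both weighted by one minus a cosine of Shift.
# Shift then lerps that sum toward Color scaled by Shift and its sine.
# The result is black at zero Shift and a slightly dimmed Color at one.
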MaterialFunction /DealershipSharedLibrary/FX/Materials/MaterialFunctions/MF_MonochromaticShift
graph {nodes [
  n1 = FunctionInput() {inputName=Color}
  n2 = Normalize(VectorInput=Constant3Vector() {constant=<rgba 1 1 1 0>})
  n3 = FunctionInput(Preview=Constant() {r=1}) {bUsePreviewValueAsDefault=true, inputName=Shift, inputType=FunctionInput_Scalar, sortPriority=1}
  n4 = OneMinus(None=Cosine(None=@n3) {period=3.8})
  n5 = Multiply(A=Multiply(A=DotProduct(A=@n1, B=@n2), B=@n2), B=@n4)
  n6 = Multiply(A=@n4, B=@n1)
  n7 = Sine(None=@n3) {period=6}
  n8 = Multiply(A=@n3, B=@n1)
  FunctionOutput(None=LinearInterpolate(A=Add(A=@n5, B=@n6), B=Multiply(A=@n7, B=@n8), Alpha=@n3))
 ]}
