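# MF_Simple_Posterize is a material function exposed to the material library as "DLR Simple Posterize" under the Dealership category, which quantizes an input value into discrete steps.
# It multiplies Input by the Posterize value, applies Ceil, then divides by Posterize again.
# Rounding up rather than down means fractional values snap to the upper edge of each step.
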MaterialFunction /DealershipSharedLibrary/FX/Materials/MaterialFunctions/MF_Simple_Posterize
graph {
 properties {bExposeToLibrary=true, userExposedCaption=DLR Simple Posterize, libraryCategoriesText [Dealership]}
 nodes [
  n1 = FunctionInput() {inputName=Posterize, previewValue=<2 2 2 1>}
  n2 = Multiply(A=FunctionInput() {inputName=Input, previewValue=<1 1 1 1>}, B=@n1) {constA=1}
  FunctionOutput(None=Divide(A=Ceil(None=@n2), B=@n1))
 ]
}
