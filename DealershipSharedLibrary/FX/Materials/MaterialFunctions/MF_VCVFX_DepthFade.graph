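# Material function "DLR Depth Fade", exposed to the library under Dealership, that applies depth fade to an opacity value for FX materials, except on low-end devices.
# It multiplies the incoming opacity by a DepthFade driven by a fade distance input with a built-in default.
# A Shading Path Switch outputs the faded result on the default and deferred paths and passes opacity through unchanged on mobile.
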MaterialFunction /DealershipSharedLibrary/FX/Materials/MaterialFunctions/MF_VCVFX_DepthFade
graph {
 properties {
  bExposeToLibrary=true
  description=Dealership Disable Depth Fade on Low en devices.
  previewBlendMode=BLEND_Translucent
  userExposedCaption=DLR Depth Fade
  libraryCategoriesText [Dealership]
 }
 nodes [
  n1 = FunctionInput() {inputName=Opacity In, inputType=FunctionInput_Scalar, previewValue=<0 0 0 0>}
  n2 = Constant() {r=100}
  n3 = Multiply(
   A=@n1
   B=DepthFade(FadeDistance=FunctionInput(Preview=@n2) {bUsePreviewValueAsDefault=true, inputName=DepthFade Distance, inputType=FunctionInput_Scalar, sortPriority=1}) {fadeDistanceDefault=20}
  )
  FunctionOutput(None=ShadingPathSwitch(Default=@n3, Deferred=@n3, Mobile=@n1) {desc=Migrated from FeatureLevelSwitch})
 ]
}
