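# MF_VCVFX_LightnessShift is a library-exposed material function for vehicle cosmetics that shifts a color's lightness with a scalar Shift input.
# Shift is saturated and compared to 0.5, which returns the input Color unchanged.
# Lower values blend linearly toward black and higher values toward white, reaching each extreme at 0 and 1.
# Both blended branches are saturated before output.
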MaterialFunction /DealershipSharedLibrary/FX/Materials/MaterialFunctions/MF_VCVFX_LightnessShift
graph {
 properties {bExposeToLibrary=true, description=MF vehicle cosmetics Lightnessshift, userExposedCaption=LightnessShift}
 nodes [
  n1 = FunctionInput() {bUsePreviewValueAsDefault=true, inputName=Shift, inputType=FunctionInput_Scalar, previewValue=<1 1 1 1>}
  n2 = Saturate(None=@n1)
  n3 = Saturate(None=Subtract(A=@n2) {constB=0.5})
  n4 = FunctionInput() {bUsePreviewValueAsDefault=true, inputName=Color, sortPriority=1, previewValue=<1 1 1 1>}
  n5 = LinearInterpolate(A=@n4, Alpha=Saturate(None=Multiply(A=@n3) {constB=2}))
  n6 = LinearInterpolate(B=@n4, Alpha=Saturate(None=Multiply(A=@n2) {constB=2}))
  FunctionOutput(None=If(A=@n2, "A > B"=Saturate(None=@n5), "A == B"=@n4, "A < B"=Saturate(None=@n6)) {constB=0.5})
 ]
}
