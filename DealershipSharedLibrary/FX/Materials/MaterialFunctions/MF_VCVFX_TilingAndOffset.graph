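# VCVFX material function that produces tiled, scaled and animated UVs for effects materials in the dealership shared library.
# Texture coordinates are scaled about their center per axis, multiplied by X/Y tiling, then split so separate Scale U and Scale V inputs multiply each channel.
# The recombined UVs feed a Panner with X/Y speed, where Panner Time scales the clock; offset comes only from this panning.
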MaterialFunction /DealershipSharedLibrary/FX/Materials/MaterialFunctions/MF_VCVFX_TilingAndOffset
graph {nodes [
  n1 = TextureCoordinate()
  n2 = Constant() {r=1}
  n3 = FunctionInput(Preview=@n2) {bUsePreviewValueAsDefault=true, inputName=ScaleUVByCenter X, inputType=FunctionInput_Scalar, sortPriority=-1, previewValue=<1 1 1 1>}
  n4 = FunctionInput(Preview=@n2) {bUsePreviewValueAsDefault=true, inputName=ScaleUVByCenter Y, inputType=FunctionInput_Scalar, previewValue=<1 1 1 1>}
  n5 = MaterialFunctionCall(X=@n3, Y=@n4) {materialFunction=&MakeFloat2}
  n6 = FunctionInput(Preview=@n2) {bUsePreviewValueAsDefault=true, inputName=Tiling X, inputType=FunctionInput_Scalar, sortPriority=1, previewValue=<1 1 1 1>}
  n7 = FunctionInput(Preview=@n2) {bUsePreviewValueAsDefault=true, inputName=Tiling Y, inputType=FunctionInput_Scalar, sortPriority=2}
  n8 = MaterialFunctionCall(Float2=Multiply(
    A=MaterialFunctionCall(UVs=@n1, Texture Scale=@n5.Result) {materialFunction=&ScaleUVsByCenter}.UVs
    B=MaterialFunctionCall(X=@n6, Y=@n7) {materialFunction=&MakeFloat2}.Result
   )) {materialFunction=&BreakOutFloat2Components}
  n9 = FunctionInput() {
   bUsePreviewValueAsDefault=true
   description=Multiply the mask of U
   inputName=Scale U
   inputType=FunctionInput_Scalar
   sortPriority=6
   previewValue=<1 1 1 1>
  }
  n10 = Multiply(A=@n8.R, B=@n9)
  n11 = FunctionInput() {
   bUsePreviewValueAsDefault=true
   description=Multiply the Mask of V
   inputName=Scale V
   inputType=FunctionInput_Scalar
   sortPriority=7
   previewValue=<1 1 1 1>
  }
  n12 = Multiply(A=@n8.G, B=@n11)
  n13 = FunctionInput(Preview=@n2) {bUsePreviewValueAsDefault=true, inputName=Panner Time, inputType=FunctionInput_Scalar, sortPriority=3, previewValue=<1 1 1 1>}
  n14 = FunctionInput(Preview=@n2) {bUsePreviewValueAsDefault=true, inputName=Panner Speed X, inputType=FunctionInput_Scalar, sortPriority=4, previewValue=<1 1 1 1>}
  n15 = FunctionInput(Preview=@n2) {bUsePreviewValueAsDefault=true, inputName=Panner Speed Y, inputType=FunctionInput_Scalar, sortPriority=5}
  FunctionOutput(None=Panner(
    Coordinate=MaterialFunctionCall(X=@n10, Y=@n12) {materialFunction=&MakeFloat2}.Result
    Time=MaterialFunctionCall(Speed=@n13) {materialFunction=&TimeWithSpeedVariable}.Time
    Speed=MaterialFunctionCall(X=@n14, Y=@n15) {materialFunction=&MakeFloat2}.Result
   ))
 ]}
