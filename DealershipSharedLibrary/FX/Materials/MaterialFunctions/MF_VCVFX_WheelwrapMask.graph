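# Library-exposed material function that builds a hard-edged band mask along the V texture coordinate, named for wheel-wrap VFX.
# A period-2 sine of V forms a gradient peaking mid-range between 0 and 1.
# The Power EXP input shapes the falloff and the Multiply input scales it.
# Rounding the result turns the gradient into stepped values, so both inputs control the band's width.
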MaterialFunction /DealershipSharedLibrary/FX/Materials/MaterialFunctions/MF_VCVFX_WheelwrapMask
graph {
 properties {bExposeToLibrary=true}
 nodes [
  n1 = TextureCoordinate()
  n2 = FunctionInput() {bUsePreviewValueAsDefault=true, inputName=Power EXP, inputType=FunctionInput_Scalar, previewValue=<1 0 0 1>}
  n3 = Power(Base=Sine(None=ComponentMask(None=@n1) {g=true}) {period=2}, Exp=@n2) {constExponent=0.98}
  n4 = FunctionInput() {bUsePreviewValueAsDefault=true, inputName=Multiply, inputType=FunctionInput_Scalar, previewValue=<1 0 0 1>}
  FunctionOutput(None=Round(None=Multiply(A=@n3, B=@n4)))
 ]
}
